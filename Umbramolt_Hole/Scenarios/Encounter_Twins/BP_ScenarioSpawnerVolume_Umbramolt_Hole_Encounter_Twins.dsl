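; Scenario spawner volume for the Umbramolt Hole Twins encounter that allows encounter spawning only where a rift gate spawns.
; On BeginPlay it disables encounter spawning, then spawns and stores a rift gate static point provider at its own transform.
; When that provider reports its first spawned actor, the volume unbinds from the event and enables encounter spawning, so volumes not selected for a gate stay inactive.
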(event EventBeginPlay
  (|CallParentFunction)
  ;[1 1 1] Only enable encounter spawning for the volumes that are actually selected for a rift gate spawn
  (CallFunction|DisableEncounterSpawning)
  (bind _returnvalue (Game|SpawnActorfromClass "/Umbramolt_Hole/Lager/BP_StaticPointProvider_Umbramolt_RiftGate.BP_StaticPointProvider_Umbramolt_RiftGate_C" (Transformation|GetActorTransform) "Undefined" "MultiplyWithRoot" 0 0))
  (bind _output_get (Variables|Default|SetPointProvider _returnvalue))
  (EventDispatchers|BindEventtoOnActorSpawned _output_get (delegate Custom|PointProvider_OnActorSpawned)))

(event Custom|PointProvider_OnActorSpawned
  (params (SpawnedActor :type Actor%20Object%20Reference))
  ;[1 1 1] Only enable encounter spawning for the volumes that are actually selected for a rift gate spawn
  (EventDispatchers|UnbindEventfromOnActorSpawned (Variables|Default|GetPointProvider) (delegate Custom|PointProvider_OnActorSpawned)) ; Only enable for the first spawned actor.
  (CallFunction|EnableEncounterSpawning))

(fn ConstructionScript
  (|CallParentFunction))
