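# StateTree for the Umbramolt Hole twins boss encounter scenario, running from AI activation through defeat to rift gate closing.
# A global task finds the RiftGate actor by tag, then states run linearly: wait, activate twins, await eliminations, brief delay, finish.
# A verb transition advances after two eliminations matching either fraternal Umbramolt tag.
# Interact variants perform activation and gate closing; the older-style non-Interact tasks and the pickup spawn are disabled.
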StateTree /Umbramolt_Hole/Scenarios/Encounter_Twins/ST_Scenario_Umbramolt_Hole_Encounter_Twins
global_tasks [
 {
  id=#1
  description=Get Scenario Actor - RiftGate
  properties {
   iD=#2
   instance {
    _structType=&GetEncounterActorStateTreeTaskInstanceData
    actorQuery {autoDescription=" ANY( QuestEncounter.UmbraMolt.Target.RiftGate )", tagDictionary [{tagName=QuestEncounter.UmbraMolt.Target.RiftGate}]}
   }
   node {name=Get Scenario Actor - RiftGate, _structType=&GetEncounterActorStateTreeTask}
  }
 }
 {id=#3, description=None, properties {iD=#4}}
]
states [
 Root {
  id=#5
  type=STATE
  children [
   #6
   #7
   #8
   #9
   #10
  ]
  tasks []
  transitions []
 }
 Initial Wait {
  id=#6
  type=STATE
  tasks [{
    id=#11
    description=Delay 0.0
    properties {
     iD=#12
     instance {_structType=&StateTreeDelayTaskInstanceData, duration=0}
     node {name=None, _structType=&StateTreeDelayTask}
    }
   }]
  transitions [{id=#13, trigger=ON_STATE_COMPLETED, state {id=#7, name=Activate Twins}}]
 }
 Activate Twins {
  id=#7
  type=STATE
  tasks [
   {
    id=#14
    description=BP Umbra Molt Activate AI
    properties {
     iD=#15
     instanceObject=&self:StateTreeEditorData_0.StateTreeState_1.BP_UmbraMolt_ActivateAI_C_0
     node {
      name=None
      _structType=&StateTreeBlueprintTaskWrapper
      bTaskEnabled=false
      taskClass=/Umbramolt_Hole/Scenarios/Tasks/BP_UmbraMolt_ActivateAI.BP_UmbraMolt_ActivateAI_C
     }
    }
   }
   {
    id=#16
    description=BP Umbra Molt Interact Activate AI
    properties {
     iD=#17
     instanceObject=&self:StateTreeEditorData_0.StateTreeState_1.BP_UmbraMolt_Interact_ActivateAI_C_0
     node {
      name=None
      _structType=&StateTreeBlueprintTaskWrapper
      taskClass=/Umbramolt_Hole/Scenarios/Tasks/BP_UmbraMolt_Interact_ActivateAI.BP_UmbraMolt_Interact_ActivateAI_C
     }
    }
   }
  ]
  transitions [{
    id=#18
    trigger=ON_STATE_COMPLETED
    state {id=#8, name=Wait for Twins to be Defeated}
   }]
 }
 Wait for Twins to be Defeated {
  id=#8
  type=STATE
  tasks [{
    id=#19
    description=Scenario Verb Transition
    properties {
     iD=#20
     instance {
      _structType=/Script/EncountersRuntime.EncounterVerbTransitionStateTreeTaskInstanceData
      bIsTriggered=false
      bUseAdvancedObjectives=false
      objectiveValueThreshold=2
      userActor=null
      objectiveVerbs [{
        _structType=/Script/FortniteGame.FortObjectiveFilter_PawnEliminated
        bAllowSelfEliminations=false
        bIncrementByDistance=false
        bRequirePawnWasEliminatedByConvertedPawn=false
        bUseDistance=false
        bUseKillStreak=false
        allowedKilledPawnTypes []
        allowedTeamAffiliations []
        author {
         bCanGainProgressForThisObjective=true
         tagQuery {autoDescription="", tokenStreamVersion=0, userDescription="", queryTokenStream [], tagDictionary []}
        }
        context {tagQuery {autoDescription="", tokenStreamVersion=0, userDescription="", queryTokenStream [], tagDictionary []}}
        distance {lowerBound {type=Inclusive, value=0}, upperBound {type=Open, value=0}}
        eliminatedPawn {tagQuery {
          autoDescription=" ANY( AISpawnerData.Type.Boss.DragonCart.UmbraMolt_Fraternal.G, AISpawnerData.Type.Boss.DragonCart.UmbraMolt_Fraternal.Y )"
          tokenStreamVersion=0
          userDescription=""
          queryTokenStream [0,1,1,2,0,1]
          tagDictionary [{tagName=AISpawnerData.Type.Boss.DragonCart.UmbraMolt_Fraternal.G}, {tagName=AISpawnerData.Type.Boss.DragonCart.UmbraMolt_Fraternal.Y}]
         }}
        eliminationSource {tagQuery {autoDescription="", tokenStreamVersion=0, userDescription="", queryTokenStream [], tagDictionary []}}
        killStreak {lowerBound {type=Inclusive, value=0}, upperBound {type=Open, value=0}}
        killer {tagQuery {autoDescription="", tokenStreamVersion=0, userDescription="", queryTokenStream [], tagDictionary []}}
        prohibitedKilledPawnTypes []
        prohibitedTeamAffiliations []
       }]
      objectives []
      onTriggered {iD=#4}
     }
     node {
      name=None
      _structType=/Script/EncountersRuntime.EncounterVerbTransitionStateTreeTask
      bCanEditConsideredForCompletion=true
      bConsideredForCompletion=true
      bHasTaskCompletionDelegateDispatcher=false
      bTaskEnabled=true
      transitionHandlingPriority=Normal
      bindingsBatch {value=65535}
      executionRuntimeTemplateIndex {value=65535}
      instanceDataHandle {index=65535, source=None, stateHandle {index=65535}}
      instanceTemplateIndex {value=65535}
      outputBindingsBatch {value=65535}
      transitionTo {name=Post Battle Wait, fallback=None, iD=#21, linkType=GotoState, stateHandle {index=65535}}
     }
    }
   }]
  transitions []
 }
 Post Battle Wait {
  id=#9
  type=STATE
  tasks [{
    id=#22
    description=Delay 1.0
    properties {
     iD=#23
     instance {_structType=&StateTreeDelayTaskInstanceData, duration=1}
     node {name=None, _structType=&StateTreeDelayTask}
    }
   }]
  transitions [{id=#24, trigger=ON_STATE_COMPLETED, state {id=#10, name=Finish Encounter}}]
 }
 Finish Encounter {
  id=#10
  type=STATE
  tasks [
   {
    id=#25
    description=Close Rift Gate
    properties {
     iD=#26
     instanceObject=&self:StateTreeEditorData_0.StateTreeState_5.BP_StateTreeTask_Umbramolt_Hole_Close_Rift_C_0
     node {
      name=None
      _structType=&StateTreeBlueprintTaskWrapper
      bTaskEnabled=false
      taskClass=/Umbramolt_Hole/Scenarios/Tasks/BP_StateTreeTask_Umbramolt_Hole_Close_Rift.BP_StateTreeTask_Umbramolt_Hole_Close_Rift_C
     }
    }
   }
   {
    id=#27
    description=Close Rift Gate
    properties {
     iD=#28
     instanceObject=&self:StateTreeEditorData_0.StateTreeState_5.BP_StateTreeTask_Umbramolt_Hole_Interact_Close_Rift_C_0
     node {
      name=None
      _structType=&StateTreeBlueprintTaskWrapper
      taskClass=/Umbramolt_Hole/Scenarios/Tasks/BP_StateTreeTask_Umbramolt_Hole_Interact_Close_Rift.BP_StateTreeTask_Umbramolt_Hole_Interact_Close_Rift_C
     }
    }
   }
   {
    id=#29
    description=Spawn Pickup
    properties {
     iD=#30
     instance {
      _structType=/Script/FortniteGame.SpawnPickupStateTreeTaskInstanceData
      instigatorActor=null
      itemDefinition=/Umbramolt_Kipper/Gameplay/WID_Umbramolt_Kipper_SR
      lootTierName=None
      spawnedPickup=null
      targetActor=null
      userActor=null
      flingConeAngle {registryType=None, value=0, curve {curveTable=null, rowName=None}}
      flingDirection=<0 0 1>
      flingMagnitude {registryType=None, value=0, curve {curveTable=null, rowName=None}}
      spawnOffset=<0 0 0>
      spawnedPickupActors []
     }
     node {
      name=None
      _structType=/Script/FortniteGame.SpawnPickupStateTreeTask
      bCanEditConsideredForCompletion=true
      bConsideredForCompletion=true
      bHasTaskCompletionDelegateDispatcher=false
      bTaskEnabled=false
      transitionHandlingPriority=Normal
      bindingsBatch {value=65535}
      executionRuntimeTemplateIndex {value=65535}
      instanceDataHandle {index=65535, source=None, stateHandle {index=65535}}
      instanceTemplateIndex {value=65535}
      outputBindingsBatch {value=65535}
     }
    }
   }
  ]
  transitions [{id=#31, trigger=ON_STATE_COMPLETED, state {id=#3, name=Root, linkType=SUCCEEDED}}]
 }
]
